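# Unlit, two-sided translucent flame effect material for the VinderTech Glider Headset, usable on Niagara and skeletal meshes.
# Distorted panning noises, sine-based shape and split masks, and a detail texture form one intensity driving curve-atlas color and opacity.
# Opacity adds a pre-skinned-position dissolve, depth and view-angle fades, and optional sphere mask; emissive adds highlights and time-of-day tint.
# World position offset combines noise waves, normal push and optional bending, using Niagara dynamic parameters.
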Material /Game/Gadgets/Assets/VinderTech_GliderChute/Glider_Headset/FX/M_Glider_Headset_Flame
graph {
 properties {
  bUsedWithMeshDeformer=true
  bUsedWithNiagaraMeshParticles=true
  bUsedWithNiagaraRibbons=true
  bUsedWithNiagaraSprites=true
  bUsedWithSkeletalMesh=true
  blendMode=BLEND_Translucent
  pixelDepthOffsetMode=PDOM_Legacy
  shadingModel=MSM_Unlit
  twoSided=true
  usedShadingModels=Unlit
 }
 nodes [
  n1 = VectorParameter() {group=Emissive, parameterName=Highlight Color UV, defaultValue=<rgba 2 0.5 0 1>}
  n2 = AppendVector(A=@n1.R, B=@n1.R)
  n3 = TextureCoordinate()
  n4 = TextureSampleParameter2D!(UVs=Panner(Coordinate=Multiply(A=@n3, B=@n2), Speed=AppendVector(A=@n1.B, B=@n1.B)), Bias, Apply View MipBias) {
   group=Emissive
   mipValueMode=TMVM_MipBias
   parameterName=Highlight Color Tex
   samplerType=SAMPLERTYPE_LinearGrayscale
   texture=/Game/Weapons/FORT_Melee/Pickaxe_Cube_Ninja_Male/FX/T_FluidFlameTile
  }
  n5 = Power(Base=@n4.RGB, Exp=ScalarParameter() {defaultValue=4, group=Emissive, parameterName=Highlight Color Power}) {constExponent=4}
  n6 = ScalarParameter() {defaultValue=0.8, parameterName=Highlight Color Desaturation}
  n7 = ScalarParameter() {defaultValue=1.4, group=Frame Split, parameterName=Frame Split Offset A}
  n8 = MaterialFunctionCall(Float2=TextureCoordinate()) {materialFunction=&BreakOutFloat2Components}
  n9 = OneMinus(None=Sine(None=Multiply(A=@n7, B=@n8.R) {constB=1.5}) {period=2})
  n10 = ScalarParameter() {defaultValue=20, group=Frame Split, parameterName=Frame Split Thin A}
  n11 = DynamicParameter() {defaultValue=<rgba 1 1 1 100>, paramNames [Random Offset, Dissolve Progress, WPO Wave Intensity, Sphere Mask Range]}
  n12 = VectorParameter() {group=Frame Split, parameterName=Frame Split Distortion UV, defaultValue=<rgba 1 2 0 1>}
  n13 = TextureCoordinate()
  n14 = Multiply(A=@n13, B=AppendVector(A=@n12.R, B=@n12.R))
  n15 = AppendVector(A=@n12.B, B=@n12.B)
  n16 = TextureSample(UVs=Add(A=AppendVector(A=@n11.Random Offset, B=@n11.Random Offset), B=Panner(Coordinate=@n14, Speed=@n15))) {samplerType=SAMPLERTYPE_Normal, texture=&T_Weap_Distort_N_02}
  n17 = ScalarParameter() {defaultValue=1.5, group=Frame Split, parameterName=Frame Split Distortion Intensity}
  n18 = ComponentMask(None=TextureCoordinate()) {g=true}
  n19 = LinearInterpolate(Alpha=Multiply(A=Multiply(A=@n16.R, B=@n17), B=OneMinus(None=@n18))) {constA=0.5, constB=1.5}
  n20 = Add(
   A=Power(Base=Multiply(A=@n9, B=@n10), Exp=@n19) {constExponent=0.5}
   B=Power(Base=@n8.G, Exp=ScalarParameter() {defaultValue=4, group=Frame Split, parameterName=Frame Split Depth A})
  )
  n21 = ScalarParameter() {defaultValue=0.8, group=Frame Split, parameterName=Frame Split Offset B}
  n22 = OneMinus(None=Sine(None=Multiply(A=@n21, B=@n8.R) {constB=1.5}) {period=2})
  n23 = ScalarParameter() {defaultValue=15, group=Frame Split, parameterName=Frame Split Thin B}
  n24 = Add(
   A=Power(Base=Multiply(A=@n22, B=@n23), Exp=@n19) {constExponent=0.5}
   B=Power(Base=@n8.G, Exp=ScalarParameter() {defaultValue=3, group=Frame Split, parameterName=Frame Split Depth B})
  )
  n25 = Multiply(A=Saturate(None=Round(None=@n20)), B=Saturate(None=Round(None=@n24)))
  n26 = ScalarParameter() {defaultValue=0.7, group=Frame Split, parameterName=Lerp Frame Split}
  n27 = VectorParameter() {group=Frame Detail, parameterName=Frame Detail Distortion UV, defaultValue=<rgba 4 0.5 -0.25 0.75>}
  n28 = Multiply(A=@n13, B=AppendVector(A=@n27.R, B=@n27.R))
  n29 = AppendVector(A=@n27.B, B=@n27.B)
  n30 = TextureSample(UVs=Add(A=AppendVector(A=@n11.Random Offset, B=@n11.Random Offset), B=Panner(Coordinate=@n28, Speed=@n29))) {samplerType=SAMPLERTYPE_Normal, texture=&T_Weap_Distort_N_02}
  n31 = ScalarParameter() {defaultValue=0.008, group=Frame Detail, parameterName=Frame Detail Distortion Intensity}
  n32 = ComponentMask(None=TextureCoordinate()) {g=true}
  n33 = TextureSampleParameter2D!(UVs=Multiply(A=Multiply(A=@n30.G, B=@n31), B=OneMinus(None=@n32)), Apply View MipBias) {
   group=Frame Detail
   parameterName=Frame Detail Tex
   texture=&/T_Glider_Headset_Mask_01
  }
  n34 = ScalarParameter() {defaultValue=1, group=Frame Detail, parameterName=Frame Detail Power}
  n35 = Saturate(None=Multiply(A=Power(Base=@n33.G, Exp=@n34), B=ScalarParameter() {defaultValue=8, group=Frame Detail, parameterName=Frame Detail Mult}) {constB=8})
  n36 = ScalarParameter() {defaultValue=0.7, group=Frame Detail, parameterName=Lerp Frame Detail}
  n37 = DynamicParameter() {defaultValue=<rgba 1 1 1 100>, paramNames [Random Offset, Dissolve Progress, WPO Wave Intensity, Sphere Mask Range]}
  n38 = VectorParameter() {group=Noise, parameterName=Noise B UV, sortPriority=6, defaultValue=<rgba 2 0.5 -0.1 0.5>}
  n39 = AppendVector(A=@n38.R, B=@n38.R)
  n40 = TextureCoordinate()
  n41 = Add(
   A=AppendVector(A=@n37.Random Offset, B=@n37.Random Offset)
   B=Panner(Coordinate=Multiply(A=@n40, B=@n39), Speed=AppendVector(A=@n38.B, B=@n38.B))
  )
  n42 = VectorParameter() {group=Distortion, parameterName=Distortion UV, sortPriority=1, defaultValue=<rgba 1 1 0 1>}
  n43 = Multiply(A=@n40, B=AppendVector(A=@n42.R, B=@n42.R))
  n44 = AppendVector(A=@n42.B, B=@n42.B)
  n45 = TextureSampleParameter2D(UVs=Add(A=AppendVector(A=@n37.Random Offset, B=@n37.Random Offset), B=Panner(Coordinate=@n43, Speed=@n44))) {
   group=Distortion
   parameterName="Distortion  Tex"
   samplerType=SAMPLERTYPE_Normal
   sortPriority=0
   texture=&T_Weap_Distort_N_02
  }
  n46 = OneMinus(None=ComponentMask(None=@n40) {g=true})
  n47 = ScalarParameter() {defaultValue=0.5, group=Distortion, parameterName=Distortion Gradient Lerp, sortPriority=3}
  n48 = Multiply(
   A=Multiply(A=ComponentMask(None=@n45.RGB) {g=true, r=true}, B=LinearInterpolate(B=@n46, Alpha=@n47) {constA=1, constAlpha=0})
   B=ScalarParameter() {defaultValue=0.2, group=Distortion, parameterName=Distortion Intensity, sortPriority=2}
  )
  n49 = ChannelMaskParameter(None=TextureSampleParameter2D(UVs=Add(A=@n41, B=@n48)) {group=Noise, parameterName=NoIse B Texture, sortPriority=4, texture=&T_NoisePack_01}.RGB) {group=Noise, maskChannel=Green, parameterName=Noise B Channel, sortPriority=5, defaultValue=<rgba 0 1 0 0>}
  n50 = Multiply(A=@n49.R, B=ScalarParameter() {defaultValue=1, group=Noise, parameterName=NoIse B Intensity, sortPriority=7}) {constB=10}
  n51 = AppendVector(A=@n37.Random Offset, B=@n37.Random Offset)
  n52 = VectorParameter() {group=Noise, parameterName=Noise A UV, sortPriority=2, defaultValue=<rgba 4 1 0.2 1.5>}
  n53 = Panner(Coordinate=Multiply(A=@n40, B=AppendVector(A=@n52.R, B=@n52.R)), Speed=AppendVector(A=@n52.B, B=@n52.B))
  n54 = TextureSampleParameter2D(UVs=Add(A=@n48, B=Add(A=@n51, B=@n53))) {group=Noise, parameterName=NoIse A Texture, sortPriority=0, texture=/Game/Weapons/FORT_Melee/Pickaxe_Headset/FX/T_Pickaxe_Headset_Noise_01}
  n55 = Multiply(
   A=ChannelMaskParameter(None=@n54.RGB) {group=Noise, parameterName=Noise A Channel, sortPriority=1}.R
   B=ScalarParameter() {defaultValue=1, group=Noise, parameterName=NoIse A Intensity, sortPriority=3}
  ) {constB=10}
  n56 = MaterialFunctionCall(Float2=TextureCoordinate()) {materialFunction=&BreakOutFloat2Components}
  n57 = Multiply(A=@n56.R) {constB=0.5}
  n58 = ScalarParameter() {defaultValue=3, group=Frame Shape, parameterName=Top Thickness, sortPriority=2}
  n59 = ScalarParameter() {defaultValue=0.5, group=Frame Shape, parameterName=Bottom Thickness, sortPriority=3}
  n60 = Multiply(A=Power(Base=Sine(None=@n57), Exp=LinearInterpolate(A=@n58, B=@n59, Alpha=@n56.G) {constA=6, constB=2}) {constExponent=4}, B=@n56.G)
  n61 = Add(
   A=LinearInterpolate(
    A=Saturate(None=@n50)
    B=Saturate(None=@n55)
    Alpha=ScalarParameter() {defaultValue=0.5, group=Noise, parameterName=Lerp Noise AB, sortPriority=8}
   )
   B=Saturate(None=Power(Base=@n60))
  )
  n62 = ScalarParameter() {defaultValue=1.25, group=Frame Shape, parameterName=Top Fade, sortPriority=0}
  n63 = ScalarParameter() {defaultValue=1, group=Frame Shape, parameterName=Bottom Fade, sortPriority=1}
  n64 = Multiply(
   A=Multiply(A=LinearInterpolate(B=@n25, Alpha=@n26) {constA=1, constB=0}, B=LinearInterpolate(B=@n35, Alpha=@n36) {constA=1, constB=0})
   B=Power(Base=Multiply(A=@n61, B=@n60), Exp=LinearInterpolate(A=@n62, B=@n63, Alpha=@n56.G) {constA=1, constB=0.5})
  )
  n65 = CurveAtlasRowParameter(CurveTime=@n64) {
   atlas=/Game/Athena/LinearColorCurves/A_Atlas_Turtleneck_Glider
   curve=&/Curve_Glider_Headset_Flame_01
   defaultValue=44
   group=Emissive
   parameterName=ColorCurve
   sortPriority=0
  }
  n66 = Multiply(A=@n64, B=VectorParameter() {group=Emissive, parameterName=Color Bright, defaultValue=<rgba 2 0 0.3333 1>}.RGB)
  n67 = Add(
   A=Multiply(A=Desaturation(None=@n5, Fraction=@n6), B=VectorParameter() {group=Emissive, parameterName=Highlight Color, defaultValue=<rgba 0 0 0 1>}.RGB)
   B=Multiply(A=Add(A=@n65, B=@n66))
  )
  n68 = ScalarParameter() {defaultValue=0.75, group=Emissive, parameterName=Emissive_Intensity}
  n69 = Constant3Vector() {constant=<rgba 1 1 1 0>}
  n70 = Constant3Vector() {constant=<rgba 1.5 1.5 1.5 0>}
  n71 = Constant3Vector() {constant=<rgba 2 1 1 0>}
  MaterialFunctionCall(Emissive=Multiply(
    A=Multiply(A=@n67, B=@n68)
    B=MaterialFunctionCall(Color_Morning=@n69, Color_Day=@n70, Color_Evening=@n69, Color_Night=@n71) {materialFunction=&Color_Time_of_Day}.Result
   )) {materialFunction=&MF_FX_EmissiveAdjustments}
  n72 = DynamicParameter() {desc=Parameters set by Niagara, parameterIndex=1, paramNames [X,Y,Z,WPOIntensity]}
  n73 = Subtract(
   A=MaterialFunctionCall(X=@n72.X, Y=@n72.X, Z=@n72.X) {materialFunction=&MakeFloat3}.Result
   B=ParticlePositionWS()
  )
  n74 = ScalarParameter() {defaultValue=-100, group=Interpolate WPO, parameterName=Clamp_min}
  n75 = ScalarParameter() {defaultValue=100, group=Interpolate WPO, parameterName=Clamp_max}
  n76 = TextureCoordinate()
  n77 = Power(
   Base=ComponentMask(None=OneMinus(None=@n76)) {g=true}
   Exp=ScalarParameter() {defaultValue=3, group=Interpolate WPO, parameterName=WPOBendPower}
  )
  n78 = StaticSwitchParameter(
   True=LinearInterpolate(B=Clamp(None=@n73, Min=@n74, Max=@n75) {maxDefault=1000, minDefault=-1000}, Alpha=Multiply(A=@n77, B=@n72.WPOIntensity))
   False=Constant()
  ) {group=Interpolate WPO, parameterName=Use Interpolate WPO}
  n79 = DynamicParameter() {defaultValue=<rgba 1 1 1 100>, paramNames [Random Offset, Dissolve Progress, WPO Wave Intensity, Sphere Mask Range]}
  n80 = VectorParameter() {group=WPO, parameterName=WPO Side UV, sortPriority=6, defaultValue=<rgba 1 0.5 -0.1 0.5>}
  n81 = Multiply(A=TextureCoordinate() {coordinateIndex=1}, B=AppendVector(A=@n80.R, B=@n80.R))
  n82 = AppendVector(A=@n80.B, B=@n80.B)
  n83 = TextureSample(UVs=Add(A=AppendVector(A=@n79.Random Offset, B=@n79.Random Offset), B=Panner(Coordinate=@n81, Speed=@n82))) {samplerType=SAMPLERTYPE_Normal, texture=/Game/Characters/Player/Female/Medium/Bodies/F_MED_Bistro_Astronaut/FX/T_GaussianNoise_Normal}
  n84 = Multiply(A=ComponentMask(None=@n83.RGB) {g=true, r=true}, B=ScalarParameter() {defaultValue=20, group=WPO, parameterName=WPO Side Intensity}) {constB=10}
  n85 = DynamicParameter() {defaultValue=<rgba 1 1 1 100>, paramNames [Random Offset, Dissolve Progress, WPO Wave Intensity, Sphere Mask Range]}
  n86 = AppendVector(A=@n85.Random Offset, B=@n85.Random Offset)
  n87 = VectorParameter() {group=WPO, parameterName=WPO Wave UV, sortPriority=6, defaultValue=<rgba 1 0.25 0 1>}
  n88 = Panner(Coordinate=Multiply(A=TextureCoordinate(), B=AppendVector(A=@n87.R, B=@n87.R)), Speed=AppendVector(A=@n87.B, B=@n87.B))
  n89 = Multiply(
   A=TextureSample(UVs=Add(A=@n86, B=@n88)) {texture=&T_Smooth_tile}.R
   B=ScalarParameter() {defaultValue=20, group=WPO, parameterName=WPO Wave Intensity}
  ) {constB=10}
  n90 = ComponentMask(None=TextureCoordinate()) {g=true}
  n91 = DynamicParameter() {defaultValue=<rgba 1 1 1 100>, paramNames [Random Offset, Dissolve Progress, WPO Wave Intensity, Sphere Mask Range]}
  n92 = Multiply(A=LinearInterpolate(B=AppendVector(A=@n84, B=@n89), Alpha=OneMinus(None=@n90)), B=@n91.WPO Wave Intensity) {constB=10}
  n93 = VertexNormalWS()
  n94 = ScalarParameter() {group=WPO, parameterName=VertexNormalScale}
  Reroute(None=Add(A=Add(A=@n78, B=@n92), B=Multiply(A=@n93, B=@n94)))
  n95 = CurveAtlasRowParameter(CurveTime=ComponentMask(None=TextureCoordinate()) {g=true}) {
   atlas=&A_Atlas_Turtleneck_Glider
   curve=&Curve_Glider_Headset_Flame_01
   defaultValue=44
   group=Emissive
   parameterName=ColorCurve
   sortPriority=0
  }
  n96 = VertexInterpolator(VS=MaterialFunctionCall(Float3=PreSkinnedPosition()) {materialFunction=&BreakOutFloat3Components}.G)
  n97 = Add(A=OneMinus(None=Divide(A=@n96.PS) {constB=1024})) {constB=-0.5}
  n98 = ScalarParameter() {defaultValue=1, group=Dissolve, parameterName=Dissolve Power}
  n99 = ScalarParameter() {defaultValue=4, group=Dissolve, parameterName=Dissolve Edge Hardness, sliderMin=0.5}
  n100 = DynamicParameter() {defaultValue=<rgba 1 1 1 100>, paramNames [Random Offset, Dissolve Progress, WPO Wave Intensity, Sphere Mask Range]}
  n101 = ScalarParameter() {defaultValue=1, group=Dissolve, parameterName=Dissolve Progress}
  n102 = Subtract(
   A=Multiply(A=Power(Base=@n97, Exp=@n98), B=@n99)
   B=LinearInterpolate(A=@n99, Alpha=StaticSwitchParameter(True=@n100.Random Offset, False=@n101) {group=Dissolve, parameterName=Use Dynamic Dissolve}) {constB=-1}
  )
  n103 = Multiply(
   A=Multiply(A=Multiply(A=@n95.A, B=@n64), B=Saturate(None=@n102))
   B=ScalarParameter() {defaultValue=4, group=Opacity, parameterName=Opacity Multiplier, sortPriority=1}
  ) {constB=3}
  n104 = ScalarParameter() {defaultValue=50, group=Opacity, parameterName=DepthFadeDistance}
  n105 = Multiply(
   A=ShadingPathSwitch(Default=DepthFade(Opacity=@n103, FadeDistance=@n104), Mobile=@n103) {desc=Migrated from FeatureLevelSwitch}
   B=MaterialFunctionCall("FadeSensitivity (optional)"=ScalarParameter() {defaultValue=0.1, group=Opacity, parameterName=FadeBasedOnViewAngle}) {materialFunction=&FadeBasedOnViewAngle}.Result
  )
  n106 = WorldPosition()
  n107 = ParticlePositionWS()
  n108 = DynamicParameter() {defaultValue=<rgba 1 1 1 100>, paramNames [Random Offset, Dissolve Progress, WPO Wave Intensity, Sphere Mask Range]}
  StaticSwitchParameter(
   True=Multiply(A=@n105, B=SphereMask(A=@n106.XYZ, B=@n107, Radius=@n108.Sphere Mask Range) {attenuationRadius=50, hardnessPercent=0.2})
   False=@n105
  ) {group=Opacity, parameterName=Use Sphere Mask}
 ]
}
