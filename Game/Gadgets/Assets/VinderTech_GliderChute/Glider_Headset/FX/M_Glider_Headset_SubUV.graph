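# Translucent unlit Niagara sprite material for the GliderChute headset FX, playing a SubUV flipbook that defaults to a flame texture.
# A repeating V gradient, tiled by a row-count parameter, optionally darkens emissive and fades alpha toward each tile's top.
# Emissive supports channel selection, power, desaturation and particle color before a shared adjustment function.
# Channel-masked alpha gets power, multiplier, depth fade and particle alpha; the mobile path skips depth fade.
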Material /Game/Gadgets/Assets/VinderTech_GliderChute/Glider_Headset/FX/M_Glider_Headset_SubUV
graph {
 properties {bUsedWithNiagaraSprites=true, blendMode=BLEND_Translucent, pixelDepthOffsetMode=PDOM_Legacy, shadingModel=MSM_Unlit, usedShadingModels=Unlit}
 nodes [
  n1 = TextureCoordinate()
  n2 = Frac(None=Multiply(A=ComponentMask(None=@n1) {g=true}, B=ScalarParameter() {defaultValue=8, parameterName=V Tile}))
  n3 = TextureSampleParameterSubUV!(UVs, Apply View MipBias) {parameterName=SubUV Texture, texture=&/T_Glider_Headset_Flame_02}
  n4 = ChannelMaskParameter(None=@n3.RGB) {parameterName=Emissive Channel}
  n5 = Multiply(
   A=LinearInterpolate(B=@n2, Alpha=ScalarParameter() {parameterName=Emissive Top Dark}) {constA=1}
   B=Power(
    Base=StaticSwitchParameter(True=@n4.R, False=@n3.RGB) {parameterName=Use Emissive Channel}
    Exp=ScalarParameter() {defaultValue=1, parameterName=EmissivePower}
   ) {constExponent=-0.1}
  )
  n6 = ScalarParameter() {parameterName=Desaturation}
  n7 = ParticleColor()
  MaterialFunctionCall(Emissive=Multiply(A=Desaturation(None=@n5, Fraction=@n6), B=@n7.RGB)) {materialFunction=&MF_FX_EmissiveAdjustments}
  n8 = ChannelMaskParameter(None=@n3.RGBA) {parameterName=Alpha Channel}
  n9 = LinearInterpolate(B=@n2, Alpha=ScalarParameter() {defaultValue=0.5, parameterName=Alpha Top Fade}) {constA=1}
  n10 = Multiply(
   A=Power(Base=Multiply(A=@n8.R, B=@n9), Exp=ScalarParameter() {defaultValue=1, parameterName=AlphaPower}) {constExponent=-0.1}
   B=ScalarParameter() {defaultValue=1, parameterName=Alpha Mult}
  ) {constB=0.25}
  n11 = DepthFade(
   Opacity=@n10
   FadeDistance=LinearInterpolate(A=ScalarParameter() {defaultValue=60, parameterName=DepthFade_Low}, B=ScalarParameter() {defaultValue=120, parameterName=DepthFade_High}) {constA=60, constB=120, desc=Bias Noise}
  )
  Multiply(A=@n7.A, B=ShadingPathSwitch(Default=@n11, Deferred=@n11, Mobile=@n10) {desc=Migrated from FeatureLevelSwitch})
 ]
}
